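# Sound cue for the All Knowing Commando character's reactive hand-snap sound, which the wave names mark as a kill reaction.
# A random node picks one of three equally weighted snap waves, and a modulator applies slight pitch and volume variation.
# Basic foley attenuation and the gadgets sound class apply.
# A concurrency override allows a single instance, stopping the oldest when a new one starts.
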SoundCue /BRCosmetics/Sounds/Characters/Commando_All_Kknowing/Character_Commando_AllKnowing_Reactive_Cue
root=Modulator_0
nodes [
 {id=Modulator_0, children [Random_1], properties {pitchMax=1.1, volumeMax=1.1, volumeMin=1}}
 {id=Random_1, children [WavePlayer_0,WavePlayer_1,WavePlayer_2], properties {weights [1,1,1]}}
 {id=WavePlayer_0, properties {soundWaveAssetPtr=&/KillReaction_AllKnowing_HandSnap_01}}
 {id=WavePlayer_1, properties {soundWaveAssetPtr=&/KillReaction_AllKnowing_HandSnap_02}}
 {id=WavePlayer_2, properties {soundWaveAssetPtr=&/KillReaction_AllKnowing_HandSnap_03}}
]
properties {
 attenuationSettings=&Foley_Attenuation_Basic_Att
 bOverrideConcurrency=true
 maxDistance=1536
 soundClassObject=&Fort_Gadgets
 volumeMultiplier=0.7
 concurrencyOverrides {maxCount=1, resolutionRule=StopOldest}
}
